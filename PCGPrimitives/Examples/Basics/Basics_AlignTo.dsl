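; PCG example graph demonstrating the Transform_Align primitive from the PCGPrimitives library.
; Two Create_Point subgraphs supply the points: one feeds the PrimaryInput and the other the SecondaryInput of three parallel Transform_Align subgraphs.
; All three align nodes receive identical inputs and show no differing settings.
; Their results are not wired to any further node or graph output.
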(graph Basics_AlignTo
  (bind Subgraph_0 (Subgraph :Graph /PCGPrimitives/Primitives/Create/Create_Point.Create_Point))
  (bind Subgraph_1 (Subgraph :Graph /PCGPrimitives/Primitives/Create/Create_Point.Create_Point))
  (bind Subgraph_2 (Subgraph :Graph /PCGPrimitives/Primitives/Transform/Transform_Align.Transform_Align))
  (bind Subgraph_3 (Subgraph :Graph /PCGPrimitives/Primitives/Transform/Transform_Align.Transform_Align))
  (bind Subgraph_4 (Subgraph :Graph /PCGPrimitives/Primitives/Transform/Transform_Align.Transform_Align))
  (connect Subgraph_1 Out Subgraph_2 PrimaryInput)
  (connect Subgraph_0 Out Subgraph_2 SecondaryInput)
  (connect Subgraph_1 Out Subgraph_3 PrimaryInput)
  (connect Subgraph_0 Out Subgraph_3 SecondaryInput)
  (connect Subgraph_1 Out Subgraph_4 PrimaryInput)
  (connect Subgraph_0 Out Subgraph_4 SecondaryInput))
