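; Basics_AlignToStack is a PCG example graph that builds a stack of three assets using the PCGPrimitives subgraph library.
; Three Create_Point branches each feed an Assign_Assets node to supply the pieces.
; Chained Transform_Align nodes align the second piece against the base, then the third against that aligned result.
; The base and both aligned results merge into one Spawn_Assets node.
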(graph Basics_AlignToStack
  (bind Subgraph_0 (Subgraph :Graph /PCGPrimitives/Primitives/Create/Create_Point.Create_Point))
  (bind Subgraph_1 (Subgraph :Graph /PCGPrimitives/Primitives/Create/Create_Point.Create_Point))
  (bind Subgraph_5 (Subgraph :Graph /PCGPrimitives/Primitives/Assign/Assign_Assets.Assign_Assets))
  (bind Subgraph_6 (Subgraph :Graph /PCGPrimitives/Primitives/Assign/Assign_Assets.Assign_Assets))
  (bind Subgraph_2 (Subgraph :Graph /PCGPrimitives/Primitives/Transform/Transform_Align.Transform_Align))
  (bind Subgraph_4 (Subgraph :Graph /PCGPrimitives/Primitives/Create/Create_Point.Create_Point))
  (bind Subgraph_7 (Subgraph :Graph /PCGPrimitives/Primitives/Assign/Assign_Assets.Assign_Assets))
  (bind Subgraph_3 (Subgraph :Graph /PCGPrimitives/Primitives/Transform/Transform_Align.Transform_Align))
  (bind Subgraph_8 (Subgraph :Graph /PCGPrimitives/Primitives/Spawn/Spawn_Assets.Spawn_Assets))
  (connect Subgraph_5 Out Subgraph_2 PrimaryInput)
  (connect Subgraph_6 Out Subgraph_2 SecondaryInput)
  (connect Subgraph_7 Out Subgraph_3 PrimaryInput)
  (connect Subgraph_2 Out Subgraph_3 SecondaryInput)
  (connect Subgraph_1 Out Subgraph_5 PrimaryInput)
  (connect Subgraph_0 Out Subgraph_6 PrimaryInput)
  (connect Subgraph_4 Out Subgraph_7 PrimaryInput)
  (connect Subgraph_2 Out Subgraph_8 PrimaryInput)
  (connect Subgraph_3 Out Subgraph_8 PrimaryInput)
  (connect Subgraph_6 Out Subgraph_8 PrimaryInput))
