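; PCG example graph that builds arena bleachers, with concrete steps and rows of chairs, between an inner and an outer spline shape.
; Trace_Splines_ParametricGrid interpolates span splines between the two shapes, with BleacherRows setting the transition count.
; Two Place_OnSpline branches put concrete and chairs (spaced by ChairSpacing) on those splines.
; Each branch zeroes roll and pitch to discard tilt from the inter-spline direction, then assigns and spawns its parameterized assets.
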(graph Basics_ArenaBleachersWithParametricGrid
  (params (ConcreteAsset :type (Array (Struct /Script/CoreUObject.SoftObjectPath)) :default [(struct :AssetPath (struct :PackageName /PCG/DebugObjects/PCG_Cube :AssetName PCG_Cube))]) (ChairAsset :type (Array (Struct /Script/CoreUObject.SoftObjectPath)) :default [(struct :AssetPath (struct :PackageName /PCGPrimitives/Meshes/SM_BackyardChair03 :AssetName SM_BackyardChair03))]) (BleacherRows :type Int :default 5) (ChairSpacing :type Double :default 1e+02))
  (bind InnerOval (Subgraph :Graph /PCGPrimitives/Primitives/Create/Create_Spline_Shape.Create_Spline_Shape))
  (bind OuterOval (Subgraph :Graph /PCGPrimitives/Primitives/Create/Create_Spline_Shape.Create_Spline_Shape))
  (bind ParametricGrid (Subgraph :Graph /PCGPrimitives/Primitives/Trace/Trace_Splines_ParametricGrid.Trace_Splines_ParametricGrid :TransitionSplines BleacherRows))
  (bind PlaceConcrete (Subgraph :Graph /PCGPrimitives/Primitives/Place/Place_OnSpline.Place_OnSpline))
  (bind PlaceChairs (Subgraph :Graph /PCGPrimitives/Primitives/Place/Place_OnSpline.Place_OnSpline :Spacing ChairSpacing))
  (bind AttributeMathsOp_0 (Attribute%20Maths%20Op :InputSource1 $Rotation.Roll :Operation Set)) ; force rotation Roll to be zero. \nWe do not want to iterate the rotation from the direction between \nthe two splines used to create the arena structure. 
  (bind AttributeMathsOp_1 (Attribute%20Maths%20Op :InputSource1 $Rotation.Pitch :Operation Set)) ; force rotation Pitch to be zero. \nWe do not want to iterate the rotation from the direction between \nthe two splines used to create the arena structure. 
  (bind AssignConcrete (Subgraph :Graph /PCGPrimitives/Primitives/Assign/Assign_Assets.Assign_Assets :Assets ConcreteAsset))
  (bind SpawnConcrete (Subgraph :Graph /PCGPrimitives/Primitives/Spawn/Spawn_Assets.Spawn_Assets))
  (bind AttributeMathsOp_2 (Attribute%20Maths%20Op :InputSource1 $Rotation.Roll :Operation Set)) ; force rotation Roll to be zero. \nWe do not want to iterate the rotation from the direction between \nthe two splines used to create the arena structure. 
  (bind AttributeMathsOp_3 (Attribute%20Maths%20Op :InputSource1 $Rotation.Pitch :Operation Set)) ; force rotation Pitch to be zero. \nWe do not want to iterate the rotation from the direction between \nthe two splines used to create the arena structure. 
  (bind AssignChairs (Subgraph :Graph /PCGPrimitives/Primitives/Assign/Assign_Assets.Assign_Assets :Assets ChairAsset))
  (bind SpawnChairs (Subgraph :Graph /PCGPrimitives/Primitives/Spawn/Spawn_Assets.Spawn_Assets))
  (connect AttributeMathsOp_3 Out AssignChairs PrimaryInput)
  (connect AttributeMathsOp_1 Out AssignConcrete PrimaryInput)
  (connect PlaceConcrete Out AttributeMathsOp_0 InA)
  (connect AttributeMathsOp_0 Out AttributeMathsOp_1 InA)
  (connect PlaceChairs Out AttributeMathsOp_2 InA)
  (connect AttributeMathsOp_2 Out AttributeMathsOp_3 InA)
  (connect InnerOval Out ParametricGrid PrimaryInput)
  (connect OuterOval Out ParametricGrid SecondaryInput)
  (connect ParametricGrid Span%20Splines PlaceChairs PrimaryInput)
  (connect ParametricGrid Span%20Splines PlaceConcrete PrimaryInput)
  (connect AssignChairs Out SpawnChairs PrimaryInput)
  (connect AssignConcrete Out SpawnConcrete PrimaryInput))
