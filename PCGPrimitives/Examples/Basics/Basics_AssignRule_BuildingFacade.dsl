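; PCG example graph that demonstrates the Assign_Rule primitive by building a facade along a spline.
; Three reusable subgraphs run in a linear chain: Create_Spline_line feeds Assign_Rule, whose output feeds Spawn_Assets.
; Assign_Rule composes facade modules along the spline following a shape grammar string, and Spawn_Assets receives that result for spawning.
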(graph Basics_AssignRule_BuildingFacade
  (bind Subgraph_0 (Subgraph :Graph /PCGPrimitives/Primitives/Create/Create_Spline_line.Create_Spline_line))
  (bind Subgraph_1 (Subgraph :Graph /PCGPrimitives/Primitives/Assign/Assign_Rule.Assign_Rule)) ; Composes facade modules along the spline\nfollowing the shape grammar string
  (bind Subgraph_2 (Subgraph :Graph /PCGPrimitives/Primitives/Spawn/Spawn_Assets.Spawn_Assets))
  (connect Subgraph_0 Out Subgraph_1 PrimaryInput)
  (connect Subgraph_1 Out Subgraph_2 PrimaryInput))
