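; PCG example graph that blocks out a stacked tower from cylinder meshes with centered pivots.
; Three Create_Point primitives define base, cap and top cylinder boxes, with Z location matching Z extent to put each bottom at ground level.
; Transform_Offset adds to the cap and top locations, then all three points merge into Assign_Assets and Spawn_Assets.
; Matching scale to input bounds, without mesh bounds or uniform scale, makes meshes fit the box dimensions.
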(graph Basics_BlockoutWithCenteredPivotAssets
  (bind BasePoint (Subgraph :Graph /PCGPrimitives/Primitives/Create/Create_Point.Create_Point)) ; Since the final spawned asset is a cylinder with a centered pivot,\nwe need to offset it so it sits on the ground. \nThe Z location therefore matches the Z extent,\neffectively placing the pivot at the bottom of the cylinder.
  (bind CapPoint (Subgraph :Graph /PCGPrimitives/Primitives/Create/Create_Point.Create_Point)) ; Since the final spawned asset is a cylinder with a centered pivot,\nwe need to offset it so it sits on the ground. \nThe Z location therefore matches the Z extent,\neffectively placing the pivot at the bottom of the cylinder.
  (bind OffsetCap (Subgraph :Graph /PCGPrimitives/Primitives/Transform/Transform_Offset.Transform_Offset)) ; Transform_Offset - **ADDS** to the Location
  (bind TopCylinderPoint (Subgraph :Graph /PCGPrimitives/Primitives/Create/Create_Point.Create_Point)) ; Since the final spawned asset is a cylinder with a centered pivot,\nwe need to offset it so it sits on the ground. \nThe Z location therefore matches the Z extent,\neffectively placing the pivot at the bottom of the cylinder.
  (bind OffsetTopCylinder (Subgraph :Graph /PCGPrimitives/Primitives/Transform/Transform_Offset.Transform_Offset)) ; Transform_Offset - **ADDS** to the Location
  (bind AssignBaseCylinder (Subgraph :Graph /PCGPrimitives/Primitives/Assign/Assign_Assets.Assign_Assets)) ; We disable Use Mesh Bounds and enable Match Scale to Input Bounds,\n so the final asset exactly matches the input box dimensions.
  (bind SpawnTower (Subgraph :Graph /PCGPrimitives/Primitives/Spawn/Spawn_Assets.Spawn_Assets)) ; For the final spawn, disable Uniform Scale and enable Match Scale to Input Bounds \nso the asset matches the input sizes—this is essential for blockouts.
  (connect BasePoint Out AssignBaseCylinder PrimaryInput)
  (connect OffsetCap Out AssignBaseCylinder PrimaryInput)
  (connect OffsetTopCylinder Out AssignBaseCylinder PrimaryInput)
  (connect CapPoint Out OffsetCap PrimaryInput)
  (connect TopCylinderPoint Out OffsetTopCylinder PrimaryInput)
  (connect AssignBaseCylinder Out SpawnTower PrimaryInput))
