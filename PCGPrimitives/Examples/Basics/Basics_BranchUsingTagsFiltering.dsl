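; Example PCG graph from the PCGPrimitives basics showing how to branch on isolated data, such as one building footprint, using tag filtering.
; A line spline feeds a spline shape, then Write_Class tags each output as Footprints_ID:# for identification with the query isolation tool.
; Filter_Tags routes the chosen ID to Selected for random area filling, while Unselected data gets center point placement.
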(graph Basics_BranchUsingTagsFiltering
  (bind Subgraph_2 (Subgraph :Graph /PCGPrimitives/Primitives/Create/Create_Spline_line.Create_Spline_line))
  (bind Subgraph_0 (Subgraph :Graph /PCGPrimitives/Primitives/Create/Create_Spline_Shape.Create_Spline_Shape))
  (bind Subgraph_1 (Subgraph :Graph /PCGPrimitives/Primitives/Write/Write_Class.Write_Class)) ; A unique class parameter is used in a write class primitive to identify\nits the output data when using the isolation tool.\nIn this example, the unique class set is Footprints and a tag on the output \n(Footprints_ID:#, where the # correspond to the data index)\nThe debug display is enabled BEFORE launching the isolation tool \nto create a spline component of the output shape that can be easily selected \nwith the query isolation tool. The tags found from this query are then \nused for the filtering.
  (bind Subgraph_3 (Subgraph :Graph /PCGPrimitives/Primitives/Filter/Filter_Tags.Filter_Tags)) ; Filtered Input Tags used is set with Class_ID:# tag returned by the query\nisolation tool. In this example, the tag used is Footprints_ID:1, which is branched \ninto the Selected output, while all other data are branched out into the Unselected output.\nThe selected and unselected outputs are used to branch the graph into \n2 distinct sets of subsequent operations i.e. branching on an isolated building \nfootprint to allow for full customization of the building produced while maintaining \nthe existing logic on remaining unselected footprint.\nWhen inserted into a graph, the previously existing edge connections MUST be \nremoved and replaced to the Selected OR Unselected outputs based on the request.
  (bind Subgraph_4 (Subgraph :Graph /PCGPrimitives/Primitives/Fill/Fill_Areas_Randomly.Fill_Areas_Randomly))
  (bind Subgraph_9 (Subgraph :Graph /PCGPrimitives/Primitives/Place/Place_Point_AtCenter.Place_Point_AtCenter))
  (connect Subgraph_2 Out Subgraph_0 PrimaryInput)
  (connect Subgraph_0 Out Subgraph_1 PrimaryInput)
  (connect Subgraph_1 Out Subgraph_3 PrimaryInput)
  (connect Subgraph_3 Selected Subgraph_4 PrimaryInput)
  (connect Subgraph_3 Unselected Subgraph_9 PrimaryInput))
