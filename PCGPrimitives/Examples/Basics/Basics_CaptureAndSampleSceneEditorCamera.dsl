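; Example PCG graph that captures the scene from the editor camera and displays the result for debugging.
; A Get_EditorCamera subgraph feeds Capture_Scene_withPoints, which captures along the camera's forward vector.
; Its output goes to a Debug_PrimitiveData subgraph.
; The graph's note says UseInputBoundsAsCaptureSource must be disabled so the capture uses custom extents and max depth rather than input bounds.
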(graph Basics_CaptureAndSampleSceneEditorCamera
  (bind Subgraph_0 (Subgraph :Graph /PCGPrimitives/Primitives/Get/Get_EditorCamera.Get_EditorCamera))
  (bind Subgraph_1 (Subgraph :Graph /PCGPrimitives/Primitives/Capture/Capture_Scene_withPoints.Capture_Scene_withPoints)) ; Capture the scene with the camera forward vector. The UseInputBoundsAsCaptureSource paramteter must be disabled to use the custom capture extents and max depth defined by the user's prompt.
  (bind Subgraph_3 (Subgraph :Graph /PCGPrimitives/Primitives/Debug/Debug_PrimitiveData.Debug_PrimitiveData))
  (connect Subgraph_0 Out Subgraph_1 PrimaryInput)
  (connect Subgraph_1 Out Subgraph_3 PrimaryInput))
