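; Basics_CaptureAndSampleSceneFromBounds is a PCG example graph from PCGPrimitives that demonstrates capturing and sampling the scene, going by its name, from bounds.
; It chains two reusable subgraphs: Create_Point produces point data that feeds the primary input of Capture_Scene_withPoints.
; Both run on default settings, the capture logic lives inside the subgraph, and nothing is wired to the graph output.
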(graph Basics_CaptureAndSampleSceneFromBounds
  (bind Subgraph_0 (Subgraph :Graph /PCGPrimitives/Primitives/Create/Create_Point.Create_Point))
  (bind Subgraph_1 (Subgraph :Graph /PCGPrimitives/Primitives/Capture/Capture_Scene_withPoints.Capture_Scene_withPoints))
  (connect Subgraph_0 Out Subgraph_1 PrimaryInput))
