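; PCG example graph that builds a plaza with a tree at its center and rows of benches arranged radially around it.
; A spline shape sized by PlazaRadius yields a center point, which feeds both the tree branch and a radial scatter.
; The scatter places points along arcs using exposed start radius, row count, row spacing and arc spacing.
; Each branch assigns its parameterized assets and spawns them through reusable PCGPrimitives subgraphs.
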(graph Basics_CircularPlazaWithRadialScatter
  (params (TreeAsset :type (Array (Struct /Script/CoreUObject.SoftObjectPath)) :default [(struct :AssetPath (struct :PackageName /PCGBiomeSample/Meshes/PCG_Tree_01 :AssetName PCG_Tree_01))]) (PlazaRadius :type Double :default 1.5e+03) (BenchStartRadius :type Double :default 8e+02) (BenchRows :type Int :default 5) (BenchArcSpacing :type Double :default 4e+02) (BenchRowsEveryXCm :type Double :default 4e+02) (BenchAsset :type (Array (Struct /Script/CoreUObject.SoftObjectPath)) :default [(struct :AssetPath (struct :PackageName /PCGPrimitives/Meshes/SM_park_bench_N01 :AssetName SM_park_bench_N01))]))
  (bind PlazaShape (Subgraph :Graph /PCGPrimitives/Primitives/Create/Create_Spline_Shape.Create_Spline_Shape :Length PlazaRadius :Width PlazaRadius))
  (bind CenterPoint (Subgraph :Graph /PCGPrimitives/Primitives/Place/Place_Point_AtCenter.Place_Point_AtCenter))
  (bind RadialScatter (Subgraph :Graph /PCGPrimitives/Primitives/Place/Place_RadialScatter.Place_RadialScatter :IncrementRadiusDistance BenchRowsEveryXCm :NumberOfRows BenchRows :PointEvery_X_Cm_On_Arc BenchArcSpacing :StartRadius BenchStartRadius))
  (bind AssignTree (Subgraph :Graph /PCGPrimitives/Primitives/Assign/Assign_Assets.Assign_Assets :Assets TreeAsset))
  (bind SpawnTree (Subgraph :Graph /PCGPrimitives/Primitives/Spawn/Spawn_Assets.Spawn_Assets))
  (bind AssignBench (Subgraph :Graph /PCGPrimitives/Primitives/Assign/Assign_Assets.Assign_Assets :Assets BenchAsset))
  (bind SpawnBench (Subgraph :Graph /PCGPrimitives/Primitives/Spawn/Spawn_Assets.Spawn_Assets))
  (connect RadialScatter Out AssignBench PrimaryInput)
  (connect CenterPoint Out AssignTree PrimaryInput)
  (connect PlazaShape Out CenterPoint PrimaryInput)
  (connect CenterPoint Out RadialScatter PrimaryInput)
  (connect AssignBench Out SpawnBench PrimaryInput)
  (connect AssignTree Out SpawnTree PrimaryInput))
