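; Basics_CombineTwoShapes is an example PCG graph that combines two shapes using reusable primitive subgraphs.
; Two Create_Spline_Shape subgraphs each produce a shape, and both feed the same PrimaryInput pin of a Compose_Areas_Union subgraph.
; The union result is not wired to any further node or graph output.
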(graph Basics_CombineTwoShapes
  (bind Subgraph_0 (Subgraph :Graph /PCGPrimitives/Primitives/Create/Create_Spline_Shape.Create_Spline_Shape))
  (bind Subgraph_1 (Subgraph :Graph /PCGPrimitives/Primitives/Create/Create_Spline_Shape.Create_Spline_Shape))
  (bind Subgraph_2 (Subgraph :Graph /PCGPrimitives/Primitives/Compose/Compose_Areas_Union.Compose_Areas_Union))
  (connect Subgraph_0 Out Subgraph_2 PrimaryInput)
  (connect Subgraph_1 Out Subgraph_2 PrimaryInput))
