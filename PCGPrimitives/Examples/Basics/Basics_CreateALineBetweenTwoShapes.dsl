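; Example PCG graph from PCGPrimitives that creates a line between two generated spline shapes, built entirely from primitive subgraphs.
; The first shape is jittered and the second is sampled with Place_OnSpline.
; Each branch feeds Filter_RelativePosition alongside its shape's center point, using a direction vector to select an edge point.
; Trace_Connect takes both selected points to form the line.
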(graph Basics_CreateALineBetweenTwoShapes
  (bind Subgraph_0 (Subgraph :Graph /PCGPrimitives/Primitives/Create/Create_Spline_Shape.Create_Spline_Shape))
  (bind Subgraph_1 (Subgraph :Graph /PCGPrimitives/Primitives/Transform/Transform_Jitter.Transform_Jitter))
  (bind Subgraph_9 (Subgraph :Graph /PCGPrimitives/Primitives/Place/Place_Point_AtCenter.Place_Point_AtCenter))
  (bind Subgraph_7 (Subgraph :Graph /PCGPrimitives/Primitives/Filter/Filter_RelativePosition.Filter_RelativePosition)) ; The use of a direction vector toward the second shape allows to select the best point along the edge of the shape that the line should created from
  (bind Subgraph_2 (Subgraph :Graph /PCGPrimitives/Primitives/Create/Create_Spline_Shape.Create_Spline_Shape))
  (bind Subgraph_5 (Subgraph :Graph /PCGPrimitives/Primitives/Place/Place_OnSpline.Place_OnSpline))
  (bind Subgraph_8 (Subgraph :Graph /PCGPrimitives/Primitives/Place/Place_Point_AtCenter.Place_Point_AtCenter))
  (bind Subgraph_11 (Subgraph :Graph /PCGPrimitives/Primitives/Filter/Filter_RelativePosition.Filter_RelativePosition)) ; The use of an explicit direction vector allows to decide from which point along the edge of the shape that line should be created from
  (bind Subgraph_6 (Subgraph :Graph /PCGPrimitives/Primitives/Trace/Trace_Connect.Trace_Connect))
  (connect Subgraph_0 Out Subgraph_1 PrimaryInput)
  (connect Subgraph_2 Out Subgraph_5 PrimaryInput)
  (connect Subgraph_7 Out Subgraph_6 PrimaryInput)
  (connect Subgraph_11 Out Subgraph_6 PrimaryInput)
  (connect Subgraph_1 Out Subgraph_7 PrimaryInput)
  (connect Subgraph_9 Out Subgraph_7 SecondaryInput)
  (connect Subgraph_2 Out Subgraph_8 PrimaryInput)
  (connect Subgraph_1 Out Subgraph_9 PrimaryInput)
  (connect Subgraph_5 Out Subgraph_11 PrimaryInput)
  (connect Subgraph_8 Out Subgraph_11 SecondaryInput))
