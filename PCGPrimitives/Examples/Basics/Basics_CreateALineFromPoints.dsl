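; Example PCG graph that shows how to create a line from points using the PCGPrimitives subgraph library.
; Three Create_Point subgraphs each make a point, and two of them pass through their own Transform_Offset subgraph.
; The two offset branches and the third, untransformed point all feed the primary input of a Trace_Connect subgraph, which connects them into the line.
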(graph Basics_CreateALineFromPoints
  (bind Subgraph_1 (Subgraph :Graph /PCGPrimitives/Primitives/Create/Create_Point.Create_Point))
  (bind Subgraph_0 (Subgraph :Graph /PCGPrimitives/Primitives/Transform/Transform_Offset.Transform_Offset))
  (bind Subgraph_3 (Subgraph :Graph /PCGPrimitives/Primitives/Create/Create_Point.Create_Point))
  (bind Subgraph_2 (Subgraph :Graph /PCGPrimitives/Primitives/Transform/Transform_Offset.Transform_Offset))
  (bind Subgraph_7 (Subgraph :Graph /PCGPrimitives/Primitives/Create/Create_Point.Create_Point))
  (bind Subgraph_6 (Subgraph :Graph /PCGPrimitives/Primitives/Trace/Trace_Connect.Trace_Connect))
  (connect Subgraph_1 Out Subgraph_0 PrimaryInput)
  (connect Subgraph_3 Out Subgraph_2 PrimaryInput)
  (connect Subgraph_0 Out Subgraph_6 PrimaryInput)
  (connect Subgraph_2 Out Subgraph_6 PrimaryInput)
  (connect Subgraph_7 Out Subgraph_6 PrimaryInput))
