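; PCG example graph from PCGPrimitives that creates a shape from extracted bounds, built entirely from primitive subgraphs.
; Create_Point and Create_Spline_Shape both feed the primary input of Extract_Bounds.
; The extracted bounds then pass into Trace_Bound_Footprint, the final node in the chain, whose output is left unconnected.
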(graph Basics_CreateAShapeFromExtractedBounds
  (bind Subgraph_0 (Subgraph :Graph /PCGPrimitives/Primitives/Create/Create_Point.Create_Point))
  (bind Subgraph_1 (Subgraph :Graph /PCGPrimitives/Primitives/Create/Create_Spline_Shape.Create_Spline_Shape))
  (bind Subgraph_2 (Subgraph :Graph /PCGPrimitives/Primitives/Extract/Extract_Bounds.Extract_Bounds))
  (bind Subgraph_3 (Subgraph :Graph /PCGPrimitives/Primitives/Trace/Trace_Bound_Footprint.Trace_Bound_Footprint))
  (connect Subgraph_0 Out Subgraph_2 PrimaryInput)
  (connect Subgraph_1 Out Subgraph_2 PrimaryInput)
  (connect Subgraph_2 Out Subgraph_3 PrimaryInput))
