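; PCG example graph that exposes the editor cameras as its output.
; A single Get Editor Cameras node, with default settings, feeds straight into the graph's Out pin, with no filtering or other processing.
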(graph Basics_CreateEditorCameraActor
  (bind GetEditorCameras_0 (Get%20Editor%20Cameras))
  (output Out GetEditorCameras_0))
