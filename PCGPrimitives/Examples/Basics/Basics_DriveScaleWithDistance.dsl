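; PCG example graph, named for driving scale with distance, that chains three PCGPrimitives subgraphs.
; Create_Spline_Shape feeds Fill_Areas_Uniformly, whose output becomes the primary input of Write_DistanceFrom.
; The same spline shape also enters Write_DistanceFrom as the secondary input, as the distance reference for the fill.
; No separate scale node or parameter override appears in the graph.
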(graph Basics_DriveScaleWithDistance
  (bind Subgraph_0 (Subgraph :Graph /PCGPrimitives/Primitives/Create/Create_Spline_Shape.Create_Spline_Shape))
  (bind Subgraph_1 (Subgraph :Graph /PCGPrimitives/Primitives/Fill/Fill_Areas_Uniformly.Fill_Areas_Uniformly))
  (bind Subgraph_2 (Subgraph :Graph /PCGPrimitives/Primitives/Write/Write_DistanceFrom.Write_DistanceFrom))
  (connect Subgraph_0 Out Subgraph_1 PrimaryInput)
  (connect Subgraph_1 Out Subgraph_2 PrimaryInput)
  (connect Subgraph_0 Out Subgraph_2 SecondaryInput))
