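; PCG example graph from PCGPrimitives showing how to exclude one shape from another.
; Two Create_Spline_Shape subgraph instances each produce a shape.
; A Compose_Areas_Exclusion subgraph takes the first as its primary input and the second as its secondary input, the area to exclude.
; The exclusion result is not wired onward, and all logic lives in the referenced primitive subgraphs.
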(graph Basics_ExcludeAShapeWithAnother
  (bind Subgraph_0 (Subgraph :Graph /PCGPrimitives/Primitives/Create/Create_Spline_Shape.Create_Spline_Shape))
  (bind Subgraph_1 (Subgraph :Graph /PCGPrimitives/Primitives/Create/Create_Spline_Shape.Create_Spline_Shape))
  (bind Subgraph_3 (Subgraph :Graph /PCGPrimitives/Primitives/Compose/Compose_Areas_Exclusion.Compose_Areas_Exclusion))
  (connect Subgraph_0 Out Subgraph_3 PrimaryInput)
  (connect Subgraph_1 Out Subgraph_3 SecondaryInput))
