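; PCG example graph from the PCGPrimitives basics set that excludes shape-grammar footprints from a spline area.
; A spline line primitive passes through Assign_ShapeGrammarDefinition, whose Footprints output becomes the secondary input of Compose_Areas_Exclusion.
; A spline shape primitive supplies the primary area for that exclusion, and the result goes to Debug_PrimitiveData for inspection.
; All steps are reusable primitive subgraphs wired together.
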(graph Basics_ExcludeWithAShapeGrammarDefinition
  (bind Subgraph_0 (Subgraph :Graph /PCGPrimitives/Primitives/Create/Create_Spline_Shape.Create_Spline_Shape))
  (bind Subgraph_2 (Subgraph :Graph /PCGPrimitives/Primitives/Create/Create_Spline_line.Create_Spline_line))
  (bind Subgraph_3 (Subgraph :Graph /PCGPrimitives/Primitives/Assign/Assign_ShapeGrammarDefinition.Assign_ShapeGrammarDefinition))
  (bind Subgraph_4 (Subgraph :Graph /PCGPrimitives/Primitives/Compose/Compose_Areas_Exclusion.Compose_Areas_Exclusion))
  (bind Subgraph_1 (Subgraph :Graph /PCGPrimitives/Primitives/Debug/Debug_PrimitiveData.Debug_PrimitiveData))
  (connect Subgraph_4 Out Subgraph_1 PrimaryInput)
  (connect Subgraph_2 Out Subgraph_3 PrimaryInput)
  (connect Subgraph_0 Out Subgraph_4 PrimaryInput)
  (connect Subgraph_3 Footprints Subgraph_4 SecondaryInput))
