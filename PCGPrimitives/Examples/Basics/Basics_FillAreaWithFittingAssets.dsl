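; Example PCG graph showing how to fill an area with fitting meshes using the PCGPrimitives library.
; A Create_Spline_Shape subgraph feeds its shape into Fill_Areas_WithFittingAssets, which receives the FittingAssets parameter, defaulting to small window, wall and corner meshes.
; A note explains that small assets need a denser base point grid with smaller extents to give the fitting algorithm candidates.
; SquareWidth and Overlap parameters are declared but not visibly wired.
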(graph Basics_FillAreaWithFittingAssets
  (params (FittingAssets :type (Array (Struct /Script/CoreUObject.SoftObjectPath)) :default [(struct :AssetPath (struct :PackageName /PCGPrimitives/Meshes/L3_Window :AssetName L3_Window)) (struct :AssetPath (struct :PackageName /PCGPrimitives/Meshes/L2_Wall :AssetName L2_Wall)) (struct :AssetPath (struct :PackageName /PCGPrimitives/Meshes/L3_Corner_ :AssetName L3_Corner_))]) (SquareWidth :type Double :default 3e+03) (Overlap :type Double :default 0.85))
  (bind SquareShape (Subgraph :Graph /PCGPrimitives/Primitives/Create/Create_Spline_Shape.Create_Spline_Shape))
  (bind FillFitting (Subgraph :Graph /PCGPrimitives/Primitives/Fill/Fill_Areas_WithFittingAssets.Fill_Areas_WithFittingAssets :Assets FittingAssets)) ; Base_fill_points_per_square_meters and Base_fill_point_extents\nare overridden here because the provided assets are small.\nThe default grid density is too coarse for small meshes,\nleaving the fitting algorithm with no candidate points to place.\nIncreasing density (10 pts/m²) and shrinking extents (20cm)\nensures the base grid is fine enough for small assets to fit.
  (connect SquareShape Out FillFitting PrimaryInput))
