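; PCG example graph that fills a shape with spawned assets, named as a basic tile-fill demo of the PCGPrimitives library.
; It chains three subgraphs: Create_Spline_Shape supplies the shape, Fill_Areas_Uniformly fills it, and Spawn_Assets places assets from that result.
; All logic lives in those reusable primitives, linked Out to PrimaryInput, with no parameter overrides shown at this level.
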(graph Basics_FillWithTiles
  (bind Subgraph_0 (Subgraph :Graph /PCGPrimitives/Primitives/Create/Create_Spline_Shape.Create_Spline_Shape))
  (bind Subgraph_4 (Subgraph :Graph /PCGPrimitives/Primitives/Fill/Fill_Areas_Uniformly.Fill_Areas_Uniformly))
  (bind Subgraph_5 (Subgraph :Graph /PCGPrimitives/Primitives/Spawn/Spawn_Assets.Spawn_Assets))
  (connect Subgraph_0 Out Subgraph_4 PrimaryInput)
  (connect Subgraph_4 Out Subgraph_5 PrimaryInput))
