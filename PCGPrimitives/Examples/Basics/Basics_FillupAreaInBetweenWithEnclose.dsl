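; PCG example graph that spawns trees inside hexagon and square areas and fills the space between them with separate assets, cubes by default.
; Two spline shapes feed an Enclose step, and Areas Exclusion subtracts both shapes from that enclosure to derive the in-between area.
; Each of the three areas runs through random fill, asset assignment and spawn subgraphs.
; A debug subgraph also receives the in-between area.
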(graph Basics_FillupAreaInBetweenWithEnclose
  (params (HexagonTreeAssets :type (Array (Struct /Script/CoreUObject.SoftObjectPath)) :default [(struct :AssetPath (struct :PackageName /PCGBiomeSample/Meshes/PCG_Tree_01 :AssetName PCG_Tree_01))]) (SquareTreeAssets :type (Array (Struct /Script/CoreUObject.SoftObjectPath)) :default [(struct :AssetPath (struct :PackageName /PCGBiomeSample/Meshes/PCG_Tree_02 :AssetName PCG_Tree_02))]) (BetweenBoxAssets :type (Array (Struct /Script/CoreUObject.SoftObjectPath)) :default [(struct :AssetPath (struct :PackageName /PCG/DebugObjects/PCG_Cube :AssetName PCG_Cube))]) (HexagonWidth :type Double :default 1.5e+04) (SquareWidth :type Double :default 1.2e+04) (SquareSpacing :type Double :default 2.5e+04))
  (bind HexShape (Subgraph :Graph /PCGPrimitives/Primitives/Create/Create_Spline_Shape.Create_Spline_Shape))
  (bind SqShape (Subgraph :Graph /PCGPrimitives/Primitives/Create/Create_Spline_Shape.Create_Spline_Shape))
  (bind Enclose (Subgraph :Graph /PCGPrimitives/Primitives/Compose/Compose_Enclose.Compose_Enclose))
  (bind ExcludeHexSq (Subgraph :Graph /PCGPrimitives/Primitives/Compose/Compose_Areas_Exclusion.Compose_Areas_Exclusion))
  (bind FillHex (Subgraph :Graph /PCGPrimitives/Primitives/Fill/Fill_Areas_Randomly.Fill_Areas_Randomly))
  (bind FillSq (Subgraph :Graph /PCGPrimitives/Primitives/Fill/Fill_Areas_Randomly.Fill_Areas_Randomly))
  (bind FillBetween (Subgraph :Graph /PCGPrimitives/Primitives/Fill/Fill_Areas_Randomly.Fill_Areas_Randomly))
  (bind AssignHex (Subgraph :Graph /PCGPrimitives/Primitives/Assign/Assign_Assets.Assign_Assets :Assets HexagonTreeAssets))
  (bind AssignSq (Subgraph :Graph /PCGPrimitives/Primitives/Assign/Assign_Assets.Assign_Assets :Assets SquareTreeAssets))
  (bind AssignBetween (Subgraph :Graph /PCGPrimitives/Primitives/Assign/Assign_Assets.Assign_Assets :Assets BetweenBoxAssets))
  (bind SpawnHex (Subgraph :Graph /PCGPrimitives/Primitives/Spawn/Spawn_Assets.Spawn_Assets))
  (bind SpawnSq (Subgraph :Graph /PCGPrimitives/Primitives/Spawn/Spawn_Assets.Spawn_Assets))
  (bind SpawnBetween (Subgraph :Graph /PCGPrimitives/Primitives/Spawn/Spawn_Assets.Spawn_Assets))
  (bind Subgraph_13 (Subgraph :Graph /PCGPrimitives/Primitives/Debug/Debug_PrimitiveData.Debug_PrimitiveData))
  (connect FillBetween Out AssignBetween PrimaryInput)
  (connect FillHex Out AssignHex PrimaryInput)
  (connect FillSq Out AssignSq PrimaryInput)
  (connect HexShape Out Enclose PrimaryInput)
  (connect SqShape Out Enclose PrimaryInput)
  (connect Enclose Out ExcludeHexSq PrimaryInput)
  (connect HexShape Out ExcludeHexSq SecondaryInput)
  (connect SqShape Out ExcludeHexSq SecondaryInput)
  (connect ExcludeHexSq Out FillBetween PrimaryInput)
  (connect HexShape Out FillHex PrimaryInput)
  (connect SqShape Out FillSq PrimaryInput)
  (connect AssignBetween Out SpawnBetween PrimaryInput)
  (connect AssignHex Out SpawnHex PrimaryInput)
  (connect AssignSq Out SpawnSq PrimaryInput)
  (connect ExcludeHexSq Out Subgraph_13 PrimaryInput))
