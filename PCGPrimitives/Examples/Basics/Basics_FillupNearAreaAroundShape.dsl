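; PCG example graph that fills a spline shape, named as a hexagon, and its surrounding area with separate asset sets, by default debug cubes and trees.
; Create_Spline_Shape feeds a random fill directly and also Compose_NearbyArea, whose output gets its own random fill.
; Each branch chains reusable primitive subgraphs to assign its asset list and spawn it.
; The width and offset parameters are declared but not bound to any subgraph here.
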(graph Basics_FillupNearAreaAroundShape
  (params (HexagonBoxAssets :type (Array (Struct /Script/CoreUObject.SoftObjectPath)) :default [(struct :AssetPath (struct :PackageName /PCG/DebugObjects/PCG_Cube :AssetName PCG_Cube))]) (NearTreeAssets :type (Array (Struct /Script/CoreUObject.SoftObjectPath)) :default [(struct :AssetPath (struct :PackageName /PCGBiomeSample/Meshes/PCG_Tree_01 :AssetName PCG_Tree_01))]) (HexagonWidth :type Double :default 1.5e+04) (NearOffset :type Double :default 5e+03))
  (bind HexShape (Subgraph :Graph /PCGPrimitives/Primitives/Create/Create_Spline_Shape.Create_Spline_Shape))
  (bind NearArea (Subgraph :Graph /PCGPrimitives/Primitives/Compose/Compose_NearbyArea.Compose_NearbyArea))
  (bind FillHex (Subgraph :Graph /PCGPrimitives/Primitives/Fill/Fill_Areas_Randomly.Fill_Areas_Randomly))
  (bind FillNear (Subgraph :Graph /PCGPrimitives/Primitives/Fill/Fill_Areas_Randomly.Fill_Areas_Randomly))
  (bind AssignHex (Subgraph :Graph /PCGPrimitives/Primitives/Assign/Assign_Assets.Assign_Assets :Assets HexagonBoxAssets))
  (bind AssignNear (Subgraph :Graph /PCGPrimitives/Primitives/Assign/Assign_Assets.Assign_Assets :Assets NearTreeAssets))
  (bind SpawnHex (Subgraph :Graph /PCGPrimitives/Primitives/Spawn/Spawn_Assets.Spawn_Assets))
  (bind SpawnNear (Subgraph :Graph /PCGPrimitives/Primitives/Spawn/Spawn_Assets.Spawn_Assets))
  (connect FillHex Out AssignHex PrimaryInput)
  (connect FillNear Out AssignNear PrimaryInput)
  (connect HexShape Out FillHex PrimaryInput)
  (connect NearArea Out FillNear PrimaryInput)
  (connect HexShape Out NearArea PrimaryInput)
  (connect AssignHex Out SpawnHex PrimaryInput)
  (connect AssignNear Out SpawnNear PrimaryInput))
