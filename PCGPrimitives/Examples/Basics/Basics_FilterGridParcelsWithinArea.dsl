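; Example PCG graph that demonstrates filtering grid parcels by an area, built entirely from PCGPrimitives subgraphs.
; A hexagon spline shape is grid-subdivided into areas, then split against a circle shape into Selected and Unselected sets.
; Both sets are randomly filled, receiving TreeAssets and BoxAssets respectively before spawning, while parcels and circle also feed a debug subgraph.
; Width and cell-size parameters are declared, but no binding references them here.
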(graph Basics_FilterGridParcelsWithinArea
  (params (TreeAssets :type (Array (Struct /Script/CoreUObject.SoftObjectPath)) :default [(struct :AssetPath (struct :PackageName /PCGBiomeSample/Meshes/PCG_Tree_01 :AssetName PCG_Tree_01))]) (BoxAssets :type (Array (Struct /Script/CoreUObject.SoftObjectPath)) :default [(struct :AssetPath (struct :PackageName /PCG/DebugObjects/PCG_Cube :AssetName PCG_Cube))]) (HexagonWidth :type Double :default 3e+04) (GridCellSize :type Double :default 4e+03) (FilterCircleWidth :type Double :default 1.5e+04))
  (bind HexShape (Subgraph :Graph /PCGPrimitives/Primitives/Create/Create_Spline_Shape.Create_Spline_Shape))
  (bind SubdivideHex (Subgraph :Graph /PCGPrimitives/Primitives/Subdivide/Subdivide_Areas_withGrid.Subdivide_Areas_withGrid))
  (bind FilterCircle (Subgraph :Graph /PCGPrimitives/Primitives/Create/Create_Spline_Shape.Create_Spline_Shape))
  (bind FilterAreas (Subgraph :Graph /PCGPrimitives/Primitives/Filter/Filter_Areas_within_Areas.Filter_Areas_within_Areas))
  (bind FillSelected (Subgraph :Graph /PCGPrimitives/Primitives/Fill/Fill_Areas_Randomly.Fill_Areas_Randomly))
  (bind FillUnselected (Subgraph :Graph /PCGPrimitives/Primitives/Fill/Fill_Areas_Randomly.Fill_Areas_Randomly))
  (bind AssignTrees (Subgraph :Graph /PCGPrimitives/Primitives/Assign/Assign_Assets.Assign_Assets :Assets TreeAssets))
  (bind AssignBoxes (Subgraph :Graph /PCGPrimitives/Primitives/Assign/Assign_Assets.Assign_Assets :Assets BoxAssets))
  (bind SpawnTrees (Subgraph :Graph /PCGPrimitives/Primitives/Spawn/Spawn_Assets.Spawn_Assets))
  (bind SpawnBoxes (Subgraph :Graph /PCGPrimitives/Primitives/Spawn/Spawn_Assets.Spawn_Assets))
  (bind Subgraph_10 (Subgraph :Graph /PCGPrimitives/Primitives/Debug/Debug_PrimitiveData.Debug_PrimitiveData))
  (connect FillUnselected Out AssignBoxes PrimaryInput)
  (connect FillSelected Out AssignTrees PrimaryInput)
  (connect FilterAreas Selected FillSelected PrimaryInput)
  (connect FilterAreas Unselected FillUnselected PrimaryInput)
  (connect SubdivideHex Areas FilterAreas PrimaryInput)
  (connect FilterCircle Out FilterAreas SecondaryInput)
  (connect AssignBoxes Out SpawnBoxes PrimaryInput)
  (connect AssignTrees Out SpawnTrees PrimaryInput)
  (connect HexShape Out SubdivideHex PrimaryInput)
  (connect FilterCircle Out Subgraph_10 PrimaryInput)
  (connect SubdivideHex Areas Subgraph_10 PrimaryInput))
