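; Basics_FilterPointsBasedOnDistanceFrom is a PCG example graph that filters scattered points by their distance from a spline shape.
; A created spline shape feeds a random area fill, then serves as the secondary input to Write_DistanceFrom, which processes the filled points.
; Two chained Filter_AttributeValue subgraphs then narrow the result, the second receiving only the first's Selected output.
; All steps are reusable PCGPrimitives subgraphs.
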(graph Basics_FilterPointsBasedOnDistanceFrom
  (bind Subgraph_0 (Subgraph :Graph /PCGPrimitives/Primitives/Create/Create_Spline_Shape.Create_Spline_Shape))
  (bind Subgraph_2 (Subgraph :Graph /PCGPrimitives/Primitives/Fill/Fill_Areas_Randomly.Fill_Areas_Randomly))
  (bind Subgraph_3 (Subgraph :Graph /PCGPrimitives/Primitives/Write/Write_DistanceFrom.Write_DistanceFrom))
  (bind Subgraph_4 (Subgraph :Graph /PCGPrimitives/Primitives/Filter/Filter_AttributeValue.Filter_AttributeValue))
  (bind Subgraph_5 (Subgraph :Graph /PCGPrimitives/Primitives/Filter/Filter_AttributeValue.Filter_AttributeValue))
  (connect Subgraph_0 Out Subgraph_2 PrimaryInput)
  (connect Subgraph_2 Out Subgraph_3 PrimaryInput)
  (connect Subgraph_0 Out Subgraph_3 SecondaryInput)
  (connect Subgraph_3 Out Subgraph_4 PrimaryInput)
  (connect Subgraph_4 Selected Subgraph_5 PrimaryInput))
