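; PCG example graph that demonstrates finding data by its position relative to a point, using PCGPrimitives subgraphs.
; Two chained Create_Spline_Shape subgraphs build the shapes that Fill_Areas_Randomly fills.
; The filled result fans out to three parallel Filter_RelativePosition subgraphs, each taking the same Create_Point output as its secondary input.
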(graph Basics_FindRelativeDataFromPoint
  (bind Subgraph_9 (Subgraph :Graph /PCGPrimitives/Primitives/Create/Create_Spline_Shape.Create_Spline_Shape))
  (bind Subgraph_1 (Subgraph :Graph /PCGPrimitives/Primitives/Create/Create_Spline_Shape.Create_Spline_Shape))
  (bind Subgraph_11 (Subgraph :Graph /PCGPrimitives/Primitives/Create/Create_Point.Create_Point))
  (bind Subgraph_14 (Subgraph :Graph /PCGPrimitives/Primitives/Fill/Fill_Areas_Randomly.Fill_Areas_Randomly))
  (bind Subgraph_10 (Subgraph :Graph /PCGPrimitives/Primitives/Filter/Filter_RelativePosition.Filter_RelativePosition))
  (bind Subgraph_18 (Subgraph :Graph /PCGPrimitives/Primitives/Filter/Filter_RelativePosition.Filter_RelativePosition))
  (bind Subgraph_19 (Subgraph :Graph /PCGPrimitives/Primitives/Filter/Filter_RelativePosition.Filter_RelativePosition))
  (connect Subgraph_9 Out Subgraph_1 PrimaryInput)
  (connect Subgraph_14 Out Subgraph_10 PrimaryInput)
  (connect Subgraph_11 Out Subgraph_10 SecondaryInput)
  (connect Subgraph_1 Out Subgraph_14 PrimaryInput)
  (connect Subgraph_14 Out Subgraph_18 PrimaryInput)
  (connect Subgraph_11 Out Subgraph_18 SecondaryInput)
  (connect Subgraph_14 Out Subgraph_19 PrimaryInput)
  (connect Subgraph_11 Out Subgraph_19 SecondaryInput))
