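; Basics_FindRelativePointsFromPoint is an example PCG graph that filters scattered points by their position relative to a reference point.
; Create_Point feeds Trace_Bound_Footprint, whose output branches into Fill_Areas_Randomly and Place_Point_AtCenter.
; Filter_RelativePosition takes the random fill as its primary input and the center point as its secondary input.
; The graph consists entirely of chained PCGPrimitives subgraphs.
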(graph Basics_FindRelativePointsFromPoint
  (bind Subgraph_0 (Subgraph :Graph /PCGPrimitives/Primitives/Create/Create_Point.Create_Point))
  (bind Subgraph_1 (Subgraph :Graph /PCGPrimitives/Primitives/Trace/Trace_Bound_Footprint.Trace_Bound_Footprint))
  (bind Subgraph_2 (Subgraph :Graph /PCGPrimitives/Primitives/Fill/Fill_Areas_Randomly.Fill_Areas_Randomly))
  (bind Subgraph_3 (Subgraph :Graph /PCGPrimitives/Primitives/Place/Place_Point_AtCenter.Place_Point_AtCenter))
  (bind Subgraph_4 (Subgraph :Graph /PCGPrimitives/Primitives/Filter/Filter_RelativePosition.Filter_RelativePosition))
  (connect Subgraph_0 Out Subgraph_1 PrimaryInput)
  (connect Subgraph_1 Out Subgraph_2 PrimaryInput)
  (connect Subgraph_1 Out Subgraph_3 PrimaryInput)
  (connect Subgraph_2 Out Subgraph_4 PrimaryInput)
  (connect Subgraph_3 Out Subgraph_4 SecondaryInput))
