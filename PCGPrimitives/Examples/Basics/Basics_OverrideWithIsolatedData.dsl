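; PCG example graph showing how isolated data, extracted with an isolation tool, replaces procedural data.
; A created spline shape passes through Write_Class, whose unique class (MyArea in the comments) adds Class and ID tags to identify the output.
; Override_DataByTag sits after it to swap the input for isolated data matching those tags.
; With debug display enabled before isolation, the isolated actor holds an editable copy of the spline component.
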(graph Basics_OverrideWithIsolatedData
  (bind Subgraph_0 (Subgraph :Graph /PCGPrimitives/Primitives/Create/Create_Spline_Shape.Create_Spline_Shape))
  (bind Subgraph_1 (Subgraph :Graph /PCGPrimitives/Primitives/Write/Write_Class.Write_Class)) ; A unique class parameter is used in a write class primitive to identify its the output data when using the isolation tool.\nIn this example, the unique class set is MyArea and produces two tags on the output (Class:MyArea and MyArea_ID:#, where the # correspond to the data index)\nThe debug display is enabled BEFORE launching the isolation tool to create a spline component of the output shape in the world. The resulting isolated actor will hold a copy of this component and allow for manual interaction.
  (bind Subgraph_5 (Subgraph :Graph /PCGPrimitives/Primitives/Override/Override_DataByTag.Override_DataByTag)) ; Inserted between two primitives in order to replace the input data with the isolated data. \nFor the override to work, isolated data tags parameter must be set with the Class: and Class_ID: tags extracted from the isolation tool. In this example case, Class:MyArea, MyArea_ID:0
  (connect Subgraph_0 Out Subgraph_1 PrimaryInput)
  (connect Subgraph_1 Out Subgraph_5 PrimaryInput))
