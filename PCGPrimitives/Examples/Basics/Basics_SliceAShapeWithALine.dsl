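; A PCG example graph that slices a closed shape with a line, judging by its name and wiring.
; It builds a spline shape and a spline line from two Create primitive subgraphs.
; Both feed the Subdivide_Areas_withSpline subgraph, the shape as primary input and the line as secondary, to divide the area along the line.
; All three subgraphs keep their default settings, and the graph leaves the result unconnected to an output.
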(graph Basics_SliceAShapeWithALine
  (bind Subgraph_1 (Subgraph :Graph /PCGPrimitives/Primitives/Create/Create_Spline_Shape.Create_Spline_Shape))
  (bind Subgraph_4 (Subgraph :Graph /PCGPrimitives/Primitives/Create/Create_Spline_line.Create_Spline_line))
  (bind Subgraph_6 (Subgraph :Graph /PCGPrimitives/Primitives/Subdivide/Subdivide_Areas_withSpline.Subdivide_Areas_withSpline))
  (connect Subgraph_1 Out Subgraph_6 PrimaryInput)
  (connect Subgraph_4 Out Subgraph_6 SecondaryInput))
