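; This PCG example graph demonstrates slicing, recursively splitting and grid-subdividing a spline shape into smaller areas using PCGPrimitives subgraphs.
; A created spline shape is cut by a created spline line, then the resulting areas are recursively split and filtered by size.
; The selected areas are grid-subdivided, filtered by size again, and passed to a primitive debug subgraph.
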(graph Basics_SliceSplitAndGridSubdivideAShape
  (bind Subgraph_4 (Subgraph :Graph /PCGPrimitives/Primitives/Create/Create_Spline_Shape.Create_Spline_Shape))
  (bind Subgraph_5 (Subgraph :Graph /PCGPrimitives/Primitives/Create/Create_Spline_line.Create_Spline_line))
  (bind Subgraph_7 (Subgraph :Graph /PCGPrimitives/Primitives/Subdivide/Subdivide_Areas_withSpline.Subdivide_Areas_withSpline))
  (bind Subgraph_2 (Subgraph :Graph /PCGPrimitives/Primitives/Subdivide/Subdivide_Areas_WithRecursiveSplit.Subdivide_Areas_WithRecursiveSplit))
  (bind Subgraph_3 (Subgraph :Graph /PCGPrimitives/Primitives/Filter/Filter_Areas_BySize.Filter_Areas_BySize))
  (bind Subgraph_1 (Subgraph :Graph /PCGPrimitives/Primitives/Subdivide/Subdivide_Areas_withGrid.Subdivide_Areas_withGrid))
  (bind Subgraph_8 (Subgraph :Graph /PCGPrimitives/Primitives/Filter/Filter_Areas_BySize.Filter_Areas_BySize))
  (bind Subgraph_0 (Subgraph :Graph /PCGPrimitives/Primitives/Debug/Debug_PrimitiveData.Debug_PrimitiveData))
  (connect Subgraph_8 Selected Subgraph_0 PrimaryInput)
  (connect Subgraph_3 Selected Subgraph_1 PrimaryInput)
  (connect Subgraph_7 Areas Subgraph_2 PrimaryInput)
  (connect Subgraph_2 Areas Subgraph_3 PrimaryInput)
  (connect Subgraph_4 Out Subgraph_7 PrimaryInput)
  (connect Subgraph_5 Out Subgraph_7 SecondaryInput)
  (connect Subgraph_1 Areas Subgraph_8 PrimaryInput))
