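; PCG example graph, named as a tool, that spawns configurable assets along a spline at a set spacing.
; It reads spline data, applies a Set on Scale as scale compensation, then rebuilds a linear spline from the control points.
; The Place_OnSpline primitive subgraph distributes points on that spline and feeds Spawn_Assets, which takes the asset list and position, rotation and scale offsets.
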(graph Basics_SpawnAssetsAlongSpline_as_a_Tool
  (params (DistanceToCurve :type Float :default 5e+02) (SpawnAssets :type (Array (Struct /Script/CoreUObject.SoftObjectPath)) :default [(struct :AssetPath (struct :PackageName /PCG/DebugObjects/PCG_Cube :AssetName PCG_Cube))]) (PositionOffset :type Vector) (RotationOffset :type Rotator) (ScaleOffset :type Vector :default (1.0 1.0 1.0)))
  (bind PlaceOnSpline (Subgraph :Graph /PCGPrimitives/Primitives/Place/Place_OnSpline.Place_OnSpline :Spacing DistanceToCurve))
  (bind SpawnAssets (Subgraph :Graph /PCGPrimitives/Primitives/Spawn/Spawn_Assets.Spawn_Assets :Assets SpawnAssets :PositionOffset PositionOffset :RotationOffset RotationOffset :ScaleOffset ScaleOffset))
  ;[0.15 0.15 0.15 0.5] Scale compensation
  (bind GetSplineData_0 (Get%20Spline%20Data :bAlwaysRequeryActors true :bTrackActorsOnlyWithinBounds false))
  (bind AttributeMathsOp_0 (Attribute%20Maths%20Op :InputSource1 $Scale :Operation Set))
  (bind GetSplineControlPoints_0 (Get%20Spline%20Control%20Points))
  (bind CreateSpline_0 (Create%20Spline :bLinear true))
  (connect GetSplineData_0 Out AttributeMathsOp_0 InA)
  (connect GetSplineControlPoints_0 Out CreateSpline_0 In)
  (connect AttributeMathsOp_0 Out GetSplineControlPoints_0 In)
  (connect CreateSpline_0 Out PlaceOnSpline PrimaryInput)
  (connect PlaceOnSpline Out SpawnAssets PrimaryInput))
