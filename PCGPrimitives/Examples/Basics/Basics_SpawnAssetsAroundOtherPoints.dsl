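; PCGPrimitives basics example graph that spawns assets around other procedurally scattered points.
; A spline shape is randomly filled with seed points, then a shape created over each point is unioned into merged areas to optimize the second random fill.
; The new points get their distance from the seeds written and filtered, then written again before a random choice selects those that receive assigned assets and spawn.
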(graph Basics_SpawnAssetsAroundOtherPoints
  (bind Subgraph_0 (Subgraph :Graph /PCGPrimitives/Primitives/Create/Create_Spline_Shape.Create_Spline_Shape))
  (bind Subgraph_1 (Subgraph :Graph /PCGPrimitives/Primitives/Fill/Fill_Areas_Randomly.Fill_Areas_Randomly))
  (bind Subgraph_4 (Subgraph :Graph /PCGPrimitives/Primitives/Create/Create_Spline_Shape.Create_Spline_Shape))
  (bind Subgraph_9 (Subgraph :Graph /PCGPrimitives/Primitives/Compose/Compose_Areas_Union.Compose_Areas_Union)) ; This union operation following the creation of a shape over each point is an optimization before sampling and scattering.
  (bind Subgraph_12 (Subgraph :Graph /PCGPrimitives/Primitives/Fill/Fill_Areas_Randomly.Fill_Areas_Randomly))
  (bind Subgraph_13 (Subgraph :Graph /PCGPrimitives/Primitives/Write/Write_DistanceFrom.Write_DistanceFrom))
  (bind Subgraph_14 (Subgraph :Graph /PCGPrimitives/Primitives/Filter/Filter_AttributeValue.Filter_AttributeValue))
  (bind Subgraph_7 (Subgraph :Graph /PCGPrimitives/Primitives/Write/Write_DistanceFrom.Write_DistanceFrom))
  (bind Subgraph_8 (Subgraph :Graph /PCGPrimitives/Primitives/Filter/Filter_RandomChoice.Filter_RandomChoice))
  (bind Subgraph_10 (Subgraph :Graph /PCGPrimitives/Primitives/Assign/Assign_Assets.Assign_Assets))
  (bind Subgraph_11 (Subgraph :Graph /PCGPrimitives/Primitives/Spawn/Spawn_Assets.Spawn_Assets))
  (connect Subgraph_0 Out Subgraph_1 PrimaryInput)
  (connect Subgraph_1 Out Subgraph_4 PrimaryInput)
  (connect Subgraph_14 Selected Subgraph_7 PrimaryInput)
  (connect Subgraph_1 Out Subgraph_7 SecondaryInput)
  (connect Subgraph_7 Out Subgraph_8 PrimaryInput)
  (connect Subgraph_4 Out Subgraph_9 PrimaryInput)
  (connect Subgraph_8 Selected Subgraph_10 PrimaryInput)
  (connect Subgraph_10 Out Subgraph_11 PrimaryInput)
  (connect Subgraph_9 Out Subgraph_12 PrimaryInput)
  (connect Subgraph_12 Out Subgraph_13 PrimaryInput)
  (connect Subgraph_1 Out Subgraph_13 SecondaryInput)
  (connect Subgraph_13 Out Subgraph_14 PrimaryInput))
