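; PCG example graph that, per its name, spawns assets inside a square shape using the PCGPrimitives subgraphs.
; A linear chain creates a spline shape, fills its area randomly with points, assigns assets, filters overlapping points, then spawns the assets.
; Overlap filtering runs after asset assignment, and each step is a reusable primitive subgraph with no parameter overrides shown.
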(graph Basics_SpawnAssetsInsideASquareShape
  (bind Subgraph_0 (Subgraph :Graph /PCGPrimitives/Primitives/Create/Create_Spline_Shape.Create_Spline_Shape))
  (bind Subgraph_1 (Subgraph :Graph /PCGPrimitives/Primitives/Fill/Fill_Areas_Randomly.Fill_Areas_Randomly))
  (bind Subgraph_2 (Subgraph :Graph /PCGPrimitives/Primitives/Assign/Assign_Assets.Assign_Assets))
  (bind Subgraph_4 (Subgraph :Graph /PCGPrimitives/Primitives/Filter/Filter_Points_Overlapping.Filter_Points_Overlapping))
  (bind Subgraph_3 (Subgraph :Graph /PCGPrimitives/Primitives/Spawn/Spawn_Assets.Spawn_Assets))
  (connect Subgraph_0 Out Subgraph_1 PrimaryInput)
  (connect Subgraph_1 Out Subgraph_2 PrimaryInput)
  (connect Subgraph_4 Out Subgraph_3 PrimaryInput)
  (connect Subgraph_2 Out Subgraph_4 PrimaryInput))
