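; A PCG example graph from the PCGPrimitives basics set showing how to spawn assets sized to point bounds.
; A Create_Point subgraph makes a single box-shaped point whose size comes from its extents rather than its scale.
; That point feeds a Spawn_Assets subgraph that matches asset scale to the input bounds, with uniform scale off so each axis scales independently.
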(graph Basics_SpawnAssetsUsingPointBounds
  (bind Subgraph_0 (Subgraph :Graph /PCGPrimitives/Primitives/Create/Create_Point.Create_Point)) ; This create point will be used to create a box at a given size, here 2x2x10 meters. \nNotice how it is using its extent to drive the size and not the scale. 
  (bind Subgraph_1 (Subgraph :Graph /PCGPrimitives/Primitives/Spawn/Spawn_Assets.Spawn_Assets)) ; The spawn asset node uses the option "match scale to input bounds".\nAlso notice how the uniform scale is unchecked. Indeed, we want to make sure every axis is independently scaled. 
  (connect Subgraph_0 Out Subgraph_1 PrimaryInput))
