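; A PCG example graph, Basics_SplitAShapeRecursivelyAlongLongestEdge, showing recursive subdivision of a shape, along its longest edge according to the graph name.
; It chains three PCGPrimitives subgraphs, starting with Create_Spline_Shape as the source shape.
; That shape feeds Subdivide_Areas_WithRecursiveSplit, whose Areas output goes into Filter_Areas_BySize.
; All three subgraphs keep their usual settings, and the filtered result connects to no further node.
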(graph Basics_SplitAShapeRecursivelyAlongLongestEdge
  (bind Subgraph_4 (Subgraph :Graph /PCGPrimitives/Primitives/Create/Create_Spline_Shape.Create_Spline_Shape))
  (bind Subgraph_0 (Subgraph :Graph /PCGPrimitives/Primitives/Subdivide/Subdivide_Areas_WithRecursiveSplit.Subdivide_Areas_WithRecursiveSplit))
  (bind Subgraph_8 (Subgraph :Graph /PCGPrimitives/Primitives/Filter/Filter_Areas_BySize.Filter_Areas_BySize))
  (connect Subgraph_4 Out Subgraph_0 PrimaryInput)
  (connect Subgraph_0 Areas Subgraph_8 PrimaryInput))
