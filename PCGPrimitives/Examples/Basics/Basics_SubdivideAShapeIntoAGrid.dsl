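; Example PCG graph from PCGPrimitives that shows how to subdivide a shape into a grid of areas.
; It chains three reusable subgraphs: Create_Spline_Shape makes the shape, Subdivide_Areas_withGrid splits it, and Filter_Areas_BySize filters the resulting Areas by size.
; All three run at their usual settings, and the filtered result is not wired to anything further.
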(graph Basics_SubdivideAShapeIntoAGrid
  (bind Subgraph_4 (Subgraph :Graph /PCGPrimitives/Primitives/Create/Create_Spline_Shape.Create_Spline_Shape))
  (bind Subgraph_1 (Subgraph :Graph /PCGPrimitives/Primitives/Subdivide/Subdivide_Areas_withGrid.Subdivide_Areas_withGrid))
  (bind Subgraph_8 (Subgraph :Graph /PCGPrimitives/Primitives/Filter/Filter_Areas_BySize.Filter_Areas_BySize))
  (connect Subgraph_4 Out Subgraph_1 PrimaryInput)
  (connect Subgraph_1 Areas Subgraph_8 PrimaryInput))
