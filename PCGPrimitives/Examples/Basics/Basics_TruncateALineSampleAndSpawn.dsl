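; PCG example graph from PCGPrimitives Basics that truncates a line, places points along it and spawns assets on them.
; Two Create_Point subgraphs feed Trace_Connect to form the line, which Transform_Truncate trims before Place_OnSpline samples it for Spawn_Assets.
; The whole chain is composed of reusable PCGPrimitives subgraphs, with no parameter overrides shown.
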(graph Basics_TruncateALineSampleAndSpawn
  (bind Subgraph_1 (Subgraph :Graph /PCGPrimitives/Primitives/Create/Create_Point.Create_Point))
  (bind Subgraph_7 (Subgraph :Graph /PCGPrimitives/Primitives/Create/Create_Point.Create_Point))
  (bind Subgraph_6 (Subgraph :Graph /PCGPrimitives/Primitives/Trace/Trace_Connect.Trace_Connect))
  (bind Subgraph_0 (Subgraph :Graph /PCGPrimitives/Primitives/Transform/Transform_Truncate.Transform_Truncate))
  (bind Subgraph_2 (Subgraph :Graph /PCGPrimitives/Primitives/Place/Place_OnSpline.Place_OnSpline))
  (bind Subgraph_3 (Subgraph :Graph /PCGPrimitives/Primitives/Spawn/Spawn_Assets.Spawn_Assets))
  (connect Subgraph_6 Out Subgraph_0 PrimaryInput)
  (connect Subgraph_0 Out Subgraph_2 PrimaryInput)
  (connect Subgraph_2 Out Subgraph_3 PrimaryInput)
  (connect Subgraph_1 Out Subgraph_6 PrimaryInput)
  (connect Subgraph_7 Out Subgraph_6 PrimaryInput))
