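; Example PCG graph that visualizes surface areas by building planar meshes from a subdivided spline shape.
; A Create_Spline_Shape subgraph feeds Subdivide_Areas_withGrid, whose areas go into Filter_RandomChoice.
; The selected and unselected areas each feed a separate Create_Mesh_Planar subgraph, so the two random sets are meshed independently.
; The graph is composed entirely of reusable PCGPrimitives subgraphs.
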(graph Basics_VisualizeSurface
  (bind Subgraph_0 (Subgraph :Graph /PCGPrimitives/Primitives/Create/Create_Spline_Shape.Create_Spline_Shape))
  (bind Subgraph_1 (Subgraph :Graph /PCGPrimitives/Primitives/Subdivide/Subdivide_Areas_withGrid.Subdivide_Areas_withGrid))
  (bind Subgraph_2 (Subgraph :Graph /PCGPrimitives/Primitives/Filter/Filter_RandomChoice.Filter_RandomChoice))
  (bind Subgraph_3 (Subgraph :Graph /PCGPrimitives/Primitives/Create/Create_Mesh_Planar.Create_Mesh_Planar))
  (bind Subgraph_4 (Subgraph :Graph /PCGPrimitives/Primitives/Create/Create_Mesh_Planar.Create_Mesh_Planar))
  (connect Subgraph_0 Out Subgraph_1 PrimaryInput)
  (connect Subgraph_1 Areas Subgraph_2 PrimaryInput)
  (connect Subgraph_2 Selected Subgraph_3 PrimaryInput)
  (connect Subgraph_2 Unselected Subgraph_4 PrimaryInput))
