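; PCG example graph showing how an Assets graph parameter, an array of soft object paths, drives the asset array pin of the Spawn Assets primitive.
; A Create Point subgraph supplies a single stand-in point to the Spawn Assets subgraph's primary input.
; The asset list comes from a per-instance override on the PCGVolume actor; the parameter has no default, so nothing spawns without one.
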(graph Basics_WireParameterToAssetArrayPin
  (params (Assets :type (Array (Struct /Script/CoreUObject.SoftObjectPath))))
  (bind SpawnPoint (Subgraph :Graph /PCGPrimitives/Primitives/Create/Create_Point.Create_Point)) ; Provides one point so Spawn Assets has\nsomewhere to drop the meshes. Replace\nwith any upstream that produces points\n(scatter, grid, area fill, etc.).
  (bind SpawnAssets (Subgraph :Graph /PCGPrimitives/Primitives/Spawn/Spawn_Assets.Spawn_Assets :Assets Assets)) ; assets pin is driven by the Assets graph\nparameter via the Get Graph Parameter node\nbelow. The instance-level Assets override on\nthe PCGVolume actor populates the actual list.\n\nIf no instance override is provided, no meshes\nwill spawn — the param has no default.
  (connect SpawnPoint Out SpawnAssets PrimaryInput))
