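; PCG example graph showing how a scalar Width parameter drives vector offset pins, producing two parallel rows of spawned assets either side of a centerline spline.
; Attribute math halves Width and negates it, then Make Vector nodes build ±Y vectors with a shared zero constant.
; These feed two Transform_Offset subgraphs on the Create_Spline_line output, each feeding Spawn_Assets.
; Node comments flag pitfalls: constants output as 'None', Make Vector defaults to Vector2, unconnected pins silently kill output.
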(graph Basics_WireParameterToVectorPin
  (params (Width :type Double :default 2e+03))
  (bind CreateLine (Subgraph :Graph /PCGPrimitives/Primitives/Create/Create_Spline_line.Create_Spline_line)) ; Input centerline spline.\nWidth here is decorative — it does NOT\ndrive the offsets. Offsets come entirely\nfrom the Width graph parameter below.
  (bind Const2 (Create%20Constant :AttributeTypes (struct :DoubleValue 2.0))) ; STEP 2 — Constant divisor (value: 2).\nCreate Constant always outputs as 'None'.\nNOT '$Constant', NOT '2' — just 'None'.\nDownstream: use PCGBegin(None)PCGEnd.
  (bind DivideBy2 (Attribute%20Maths%20Op :InputSource1 Width :InputSource2 None :Operation Divide :OutputTarget HalfWidth :InA Width)) ; STEP 3 — Divide Width by 2.\nInA reads 'Width' (from GetWidth).\nInB reads 'None' (from Const2).\nOutputTarget writes result as 'HalfWidth'\nso downstream nodes can read it by name.
  (bind NegateHalf (Attribute%20Maths%20Op :InputSource1 HalfWidth :Operation Negate :OutputTarget NegHalfWidth)) ; STEP 4 — Negate the half-width.\nInputSource1 reads 'HalfWidth',\nmatching OutputTarget from DivideBy2.\nOutputTarget writes 'NegHalfWidth'\nfor the negative Make Vector below.
  (bind Const0 (Create%20Constant)) ; STEP 5 — Constant zero (value: 0).\nUsed for X and Z axes of both vectors.\nOutput attribute name is 'None'.\nMake Vector X and Z pins must use\nPCGBegin(None)PCGEnd to match it.\nALL three pins (X, Y, Z) must be\nconnected — unconnected pins yield\nno output from Make Vector.
  (bind MakeVecPos (Make%20Vector%20Attribute :InputSource1 None :InputSource2 HalfWidth :InputSource3 None :OutputType Vector)) ; STEP 6a — Build vector (0, +HalfWidth, 0).\nX = PCGBegin(None)PCGEnd  → Const0.\nY = PCGBegin(HalfWidth)PCGEnd  → DivideBy2.\nZ = PCGBegin(None)PCGEnd  → Const0.\nWARNING: ALL three pins must be connected.\nAn unconnected pin silently kills output.\nWARNING: Node defaults to Vector2 on creation.\nSet OutputType=Vector BEFORE connecting Z —\nthe Z pin does not exist until then.
  (bind MakeVecNeg (Make%20Vector%20Attribute :InputSource1 None :InputSource2 NegHalfWidth :InputSource3 None :OutputType Vector)) ; STEP 6b — Build vector (0, -HalfWidth, 0).\nX = PCGBegin(None)PCGEnd  → Const0.\nY = PCGBegin(NegHalfWidth)PCGEnd → NegateHalf.\nZ = PCGBegin(None)PCGEnd  → Const0.\nWARNING: ALL three pins must be connected.\nAn unconnected pin silently kills output.\nWARNING: Node defaults to Vector2 on creation.\nSet OutputType=Vector BEFORE connecting Z —\nthe Z pin does not exist until then.
  (bind OffsetPos (Subgraph :Graph /PCGPrimitives/Primitives/Transform/Transform_Offset.Transform_Offset)) ; STEP 7a — Offset spline to the right (+Y).\nThe 'offset' pin gets a Vector from 6a.\nHardcoded default is ignored once pin\nis connected — graph param drives it.
  (bind OffsetNeg (Subgraph :Graph /PCGPrimitives/Primitives/Transform/Transform_Offset.Transform_Offset)) ; STEP 7b — Offset spline to the left (-Y).\nThe 'offset' pin gets a Vector from 6b.\nSame pattern as Step 7a above.
  (bind SpawnPos (Subgraph :Graph /PCGPrimitives/Primitives/Spawn/Spawn_Assets.Spawn_Assets)) ; Spawns debug cubes along the right spline.\nReplace asset with your own mesh.
  (bind SpawnNeg (Subgraph :Graph /PCGPrimitives/Primitives/Spawn/Spawn_Assets.Spawn_Assets)) ; Spawns debug cubes along the left spline.\nReplace asset with your own mesh.
  (connect Const2 Out DivideBy2 InB)
  (connect Const0 Out MakeVecNeg X)
  (connect NegateHalf Out MakeVecNeg Y)
  (connect Const0 Out MakeVecNeg Z)
  (connect Const0 Out MakeVecPos X)
  (connect DivideBy2 Out MakeVecPos Y)
  (connect Const0 Out MakeVecPos Z)
  (connect DivideBy2 Out NegateHalf In)
  (connect MakeVecNeg Out OffsetNeg offset)
  (connect CreateLine Out OffsetNeg PrimaryInput)
  (connect MakeVecPos Out OffsetPos offset)
  (connect CreateLine Out OffsetPos PrimaryInput)
  (connect OffsetNeg Out SpawnNeg PrimaryInput)
  (connect OffsetPos Out SpawnPos PrimaryInput))
