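; PCG example graph that builds a hexagonal field of extruded parcels, each with its own noise-driven height.
; Create_Spline_Shape makes the hexagon, which Subdivide_Areas_withGrid cuts into closed parcels, each a separate data set.
; Write_Noise stores one value per parcel, between HeightMin and HeightMax, in the data-domain attribute CustomHeight.
; Create_Mesh_Extrude is meant to read that attribute as extrusion height instead of its built-in random range.
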(graph Basics_WriteNoiseToDataDomainAndExtrudeHeight
  (params (HexagonSize :type Double :default 3e+04) (GridParcelSize :type Double :default 4e+03) (HeightMin :type Double :default 1371.79944) (HeightMax :type Double :default 10038.142782))
  (bind Node_CreateHexagon (Subgraph :Graph /PCGPrimitives/Primitives/Create/Create_Spline_Shape.Create_Spline_Shape :Length HexagonSize :Width HexagonSize)) ; Creates a closed hexagonal spline shape\n(6-sided polygon, 300m wide).\nThis will serve as the outer boundary\nfor the grid subdivision.\nUsing Linear=true ensures hard polygon edges\nrather than curved spline interpolation.
  (bind Node_SubdivideGrid (Subgraph :Graph /PCGPrimitives/Primitives/Subdivide/Subdivide_Areas_withGrid.Subdivide_Areas_withGrid :GridLength GridParcelSize :GridWidth GridParcelSize)) ; Subdivides the hexagon into a regular\n40m x 40m grid of closed spline parcels.\nEach parcel becomes an independent data set,\nwhich is key — it allows the next step\nto write a unique noise value per parcel\nin the @Data domain.
  (bind Node_WriteNoise (Subgraph :Graph /PCGPrimitives/Primitives/Write/Write_Noise.Write_Noise :NoiseMax HeightMax :NoiseMin HeightMin)) ; Writes a random noise value into the @Data\ndomain attribute '@data.CustomHeight' on each parcel.\nWriting to @Data (rather than @Points or @Elements)\nstores a single value per data set —\none height value per parcel.\nThis is the correct domain to use\nwhen driving shape-level properties\nlike extrusion height downstream.
  (bind Node_ExtrudeMesh (Subgraph :Graph /PCGPrimitives/Primitives/Create/Create_Mesh_Extrude.Create_Mesh_Extrude)) ; Extrudes each parcel into a 3D mesh.\nIMPORTANT: the 'Height_Attribute' parameter\nis explicitly set to '@data.CustomHeight'\nto read the noise-driven height written above.\nThis overrides the built-in MinHeight/MaxHeight\nrange parameters, which would otherwise\nproduce a uniform random height.\nThis is how per-parcel height variation\nis achieved from a custom data-domain attribute.
  (connect Node_WriteNoise Out Node_ExtrudeMesh PrimaryInput)
  (connect Node_CreateHexagon Out Node_SubdivideGrid PrimaryInput)
  (connect Node_SubdivideGrid Areas Node_WriteNoise PrimaryInput))
