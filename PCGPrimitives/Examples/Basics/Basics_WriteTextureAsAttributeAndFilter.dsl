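; PCG example graph that writes texture values onto scattered points as an attribute and filters the points by that value.
; A created spline shape feeds a random area fill, whose output receives a projected texture's red channel through Write_Attribute_fromTexture.
; Filter_AttributeValue then passes only its Selected output to a debug primitive.
; Every step is a reusable PCGPrimitives subgraph.
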(graph Basics_WriteTextureAsAttributeAndFilter
  (bind Subgraph_2 (Subgraph :Graph /PCGPrimitives/Primitives/Create/Create_Spline_Shape.Create_Spline_Shape))
  (bind Subgraph_3 (Subgraph :Graph /PCGPrimitives/Primitives/Fill/Fill_Areas_Randomly.Fill_Areas_Randomly))
  (bind Subgraph_0 (Subgraph :Graph /PCGPrimitives/Primitives/Write/Write_Attribute_fromTexture.Write_Attribute_fromTexture)) ; The texture transform scale represents the extents (half of the bounds) of the projection. Only the red channel of the texture is written. It can write to a custom attribute name but also to point properties such as $Density, $Color, $Scale.
  (bind Subgraph_1 (Subgraph :Graph /PCGPrimitives/Primitives/Filter/Filter_AttributeValue.Filter_AttributeValue))
  (bind Subgraph_4 (Subgraph :Graph /PCGPrimitives/Primitives/Debug/Debug_PrimitiveData.Debug_PrimitiveData))
  (connect Subgraph_3 Out Subgraph_0 PrimaryInput)
  (connect Subgraph_0 Out Subgraph_1 PrimaryInput)
  (connect Subgraph_2 Out Subgraph_3 PrimaryInput)
  (connect Subgraph_1 Selected Subgraph_4 PrimaryInput))
